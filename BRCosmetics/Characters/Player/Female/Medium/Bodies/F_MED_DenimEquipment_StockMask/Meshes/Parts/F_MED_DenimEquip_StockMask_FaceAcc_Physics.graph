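# Physics asset for the face accessory of the female medium DenimEquipment StockMask cosmetic, defining secondary-motion bodies for its ears, fur, straps and tail.
# Spheres cover the ears, fur, tail and every base bone, while capsules form three-link left and right strap chains.
# Each ear, fur piece and the tail is constrained to its own side or part base, with both fur pieces per side branching in parallel.
# Angular limits are tightest on ears and tail, looser on straps and fur.
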PhysicsAsset /BRCosmetics/Characters/Player/Female/Medium/Bodies/F_MED_DenimEquipment_StockMask/Meshes/Parts/F_MED_DenimEquip_StockMask_FaceAcc_Physics
bodies {
 dyn_ear_l [{shapeType=Sphere, radius=5.5, shapeName=dyn_ear_l_sphere, center=<8.8648 0 0>}]
 dyn_ear_l_base [{shapeType=Sphere, radius=1, shapeName=dyn_ear_l_base_sphere}]
 dyn_ear_r [{shapeType=Sphere, radius=5.5, shapeName=dyn_ear_r_sphere, center=<8.8648 0 0>}]
 dyn_ear_r_base [{shapeType=Sphere, radius=1, shapeName=dyn_ear_r_base_sphere}]
 dyn_fur_l_01 [{shapeType=Sphere, radius=3, shapeName=dyn_fur_l_01_sphere, center=<2.8358 0 0>}]
 dyn_fur_l_02 [{shapeType=Sphere, radius=3, shapeName=dyn_fur_l_02_sphere, center=<2.8358 0 0>}]
 dyn_fur_l_base [{shapeType=Sphere, radius=5.5, shapeName=dyn_fur_l_base_sphere, center=<2.5109 0 0>}]
 dyn_fur_r_01 [{shapeType=Sphere, radius=3, shapeName=dyn_fur_r_01_sphere, center=<2.8358 0 0>}]
 dyn_fur_r_02 [{shapeType=Sphere, radius=3, shapeName=dyn_fur_r_02_sphere, center=<2.8358 0 0>}]
 dyn_fur_r_base [{shapeType=Sphere, radius=7, shapeName=dyn_fur_r_base_sphere, center=<4.6395 0.9153 0>}]
 dyn_strap_l_01 [{shapeType=Capsule, length=2, radius=2.5, shapeName=dyn_strap_l_01_capsule, center=<3.8089 0 0>, rotation=<rot 90 -179.9605 -179.9603>}]
 dyn_strap_l_02 [{shapeType=Capsule, length=2, radius=2.5, shapeName=dyn_strap_l_02_capsule, center=<3.8089 0 0>, rotation=<rot 90 0 0.0002>}]
 dyn_strap_l_03 [{shapeType=Capsule, length=2, radius=2.5, shapeName=dyn_strap_l_03_capsule, center=<3.8089 0 0>, rotation=<rot 90 -179.9998 -179.9996>}]
 dyn_strap_l_base [{shapeType=Sphere, radius=1, shapeName=dyn_strap_l_base_sphere}]
 dyn_strap_r_01 [{shapeType=Capsule, length=2, radius=2.5, shapeName=dyn_strap_r_01_capsule, center=<3.8089 0 0>, rotation=<rot 90 90 90.0003>}]
 dyn_strap_r_02 [{shapeType=Capsule, length=2, radius=2.5, shapeName=dyn_strap_r_02_capsule, center=<3.8089 0 0>, rotation=<rot 90 179.9801 179.9803>}]
 dyn_strap_r_03 [{shapeType=Capsule, length=2, radius=2.5, shapeName=dyn_strap_r_03_capsule, center=<3.8089 0 0>, rotation=<rot 90 179.9999 -179.9999>}]
 dyn_strap_r_base [{shapeType=Sphere, radius=1, shapeName=dyn_strap_r_base_sphere}]
 dyn_tail [{shapeType=Sphere, radius=3, shapeName=dyn_tail_sphere, center=<3 0 0>}]
 dyn_tail_base [{shapeType=Sphere, radius=1, shapeName=dyn_tail_base_sphere}]
}
constraints [
 {bone1Name=dyn_tail, bone2Name=dyn_tail_base, swing1LimitDegrees=5, swing2LimitDegrees=5, twistLimitDegrees=1}
 {bone1Name=dyn_ear_l, bone2Name=dyn_ear_l_base, swing1LimitDegrees=10, swing2LimitDegrees=5, twistLimitDegrees=1}
 {bone1Name=dyn_ear_r, bone2Name=dyn_ear_r_base, swing1LimitDegrees=10, swing2LimitDegrees=5, twistLimitDegrees=1}
 {bone1Name=dyn_strap_r_03, bone2Name=dyn_strap_r_02, swing1LimitDegrees=25, swing2LimitDegrees=20, twistLimitDegrees=3}
 {bone1Name=dyn_fur_l_02, bone2Name=dyn_fur_l_base, swing1LimitDegrees=30, swing2LimitDegrees=30, twistLimitDegrees=3}
 {bone1Name=dyn_fur_l_01, bone2Name=dyn_fur_l_base, swing1LimitDegrees=30, swing2LimitDegrees=30, twistLimitDegrees=3}
 {bone1Name=dyn_fur_r_02, bone2Name=dyn_fur_r_base, swing1LimitDegrees=30, swing2LimitDegrees=30, twistLimitDegrees=3}
 {bone1Name=dyn_fur_r_01, bone2Name=dyn_fur_r_base, swing1LimitDegrees=30, swing2LimitDegrees=30, twistLimitDegrees=3}
 {bone1Name=dyn_strap_l_01, bone2Name=dyn_strap_l_base, swing1LimitDegrees=25, swing2LimitDegrees=20, twistLimitDegrees=3}
 {bone1Name=dyn_strap_l_02, bone2Name=dyn_strap_l_01, swing1LimitDegrees=25, swing2LimitDegrees=20, twistLimitDegrees=3}
 {bone1Name=dyn_strap_l_03, bone2Name=dyn_strap_l_02, swing1LimitDegrees=25, swing2LimitDegrees=20, twistLimitDegrees=3}
 {bone1Name=dyn_strap_r_01, bone2Name=dyn_strap_r_base, swing1LimitDegrees=25, swing2LimitDegrees=20, twistLimitDegrees=3}
 {bone1Name=dyn_strap_r_02, bone2Name=dyn_strap_r_01, swing1LimitDegrees=25, swing2LimitDegrees=20, twistLimitDegrees=3}
]
