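# Physics asset for the face accessory part of the female medium DenimEquipment StockMask character cosmetic.
# It holds a single capsule body on the head bone, offset along the bone's X axis.
# It has no constraints, so the asset defines only that one head collision shape and no jointed bodies.
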PhysicsAsset /BRCosmetics/Characters/Player/Female/Medium/Bodies/F_MED_DenimEquipment_StockMask/Meshes/Parts/F_MED_DenimEquip_StockMask_FaceAcc_Physics1
bodies {head [{shapeType=Capsule, length=13.3254, radius=21.289, center=<11.6739 0 0>}]}
constraints []
